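# Sound cue for the one-handed sword jump attack in the melee weapon set.
# A Random node picks one of three equally weighted jump attack wave variations on each play.
# It plays in the melee weapon sound class at reduced volume, with close-range stereo melee attenuation and a global melee attack concurrency limit.
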SoundCue /CRD_Melee/Audio/Sounds/OneHandSword/Melee_1HSword_JumpAttack_Cue
root=Random_0
nodes [
 {id=Random_0, children [WavePlayer_0,WavePlayer_1,WavePlayer_2], properties {weights [1,1,1]}}
 {id=WavePlayer_0, properties {soundWaveAssetPtr=&/Melee_1HSword_JumpAttack_03}}
 {id=WavePlayer_1, properties {soundWaveAssetPtr=&/Melee_1HSword_JumpAttack_01}}
 {id=WavePlayer_2, properties {soundWaveAssetPtr=&/Melee_1HSword_JumpAttack_02}}
]
properties {
 attenuationSettings=/Game/Sounds/Attenuation/Fort_Weapon_Melee_Attack_Close_Range_Stereo_Att
 maxDistance=10058.75
 soundClassObject=&Fort_WeaponMelee
 volumeMultiplier=0.45
 concurrencySet [/CRD_Melee/Audio/Settings/Melee_Attack_Global_Conc]
}
