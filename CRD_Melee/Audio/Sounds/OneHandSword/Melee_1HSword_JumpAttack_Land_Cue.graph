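# Sound cue for the landing impact of a one-handed sword jump attack in the melee audio set.
# A Random node picks one of three landing wave variations with equal weight.
# It plays through the melee weapon sound class with close-range stereo melee attack attenuation and a shared global melee attack concurrency setting.
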SoundCue /CRD_Melee/Audio/Sounds/OneHandSword/Melee_1HSword_JumpAttack_Land_Cue
root=Random_0
nodes [
 {id=Random_0, children [WavePlayer_3,WavePlayer_4,WavePlayer_5], properties {weights [1,1,1]}}
 {id=WavePlayer_3, properties {soundWaveAssetPtr=&/Melee_1HSword_JumpAttack_Land_01}}
 {id=WavePlayer_4, properties {soundWaveAssetPtr=&/Melee_1HSword_JumpAttack_Land_02}}
 {id=WavePlayer_5, properties {soundWaveAssetPtr=&/Melee_1HSword_JumpAttack_Land_03}}
]
properties {
 attenuationSettings=/Game/Sounds/Attenuation/Fort_Weapon_Melee_Attack_Close_Range_Stereo_Att
 maxDistance=10058.75
 soundClassObject=&Fort_WeaponMelee
 concurrencySet [/CRD_Melee/Audio/Settings/Melee_Attack_Global_Conc]
}
