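# Sound cue for picking up a one-handed melee sword.
# A Random node picks one of three equally weighted pickup wave variations on each play.
# It uses the shared Fort weapon pickup attenuation and the melee weapon sound class.
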SoundCue /CRD_Melee/Audio/Sounds/OneHandSword/Melee_1HSword_Pickup_Cue
root=Random_0
nodes [
 {id=Random_0, children [WavePlayer_0,WavePlayer_1,WavePlayer_2], properties {weights [1,1,1]}}
 {id=WavePlayer_0, properties {soundWaveAssetPtr=&/Melee_1HSword_Pickup_01}}
 {id=WavePlayer_1, properties {soundWaveAssetPtr=&/Melee_1HSword_Pickup_02}}
 {id=WavePlayer_2, properties {soundWaveAssetPtr=&/Melee_1HSword_Pickup_03}}
]
properties {
 attenuationSettings=&Fort_Weapon_Pickup_Att
 maxDistance=3500
 soundClassObject=&Fort_WeaponMelee
}
