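# Sound cue for the start of the one-handed sword's secondary Whirlwind melee attack.
# A mixer layers a random pick of three Start waves with a random pick of three Repeat waves, the latter held back by a fixed delay so the spin follows the opening swing.
# It plays at reduced volume with close-range melee attack attenuation, the weapon melee sound class and a shared global melee attack concurrency setting.
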SoundCue /CRD_Melee/Audio/Sounds/OneHandSword/Melee_1HSword_Secondary_Whirlwind_Start_Cue
root=Mixer_0
nodes [
 {id=Delay_0, children [Random_1], properties {delayMax=0.65, delayMin=0.65}}
 {id=Mixer_0, children [Delay_0,Random_0], properties {inputVolume [1,1]}}
 {id=Random_0, children [WavePlayer_0,WavePlayer_1,WavePlayer_2], properties {weights [1,1,1]}}
 {id=Random_1, children [WavePlayer_3,WavePlayer_4,WavePlayer_5], properties {weights [1,1,1]}}
 {id=WavePlayer_0, properties {soundWaveAssetPtr=&/Melee_1HSword_Secondary_Whirlwind_Start_01}}
 {id=WavePlayer_1, properties {soundWaveAssetPtr=&/Melee_1HSword_Secondary_Whirlwind_Start_02}}
 {id=WavePlayer_2, properties {soundWaveAssetPtr=&/Melee_1HSword_Secondary_Whirlwind_Start_03}}
 {id=WavePlayer_3, properties {soundWaveAssetPtr=&/Melee_1HSword_Secondary_Whirlwind_Repeat_03}}
 {id=WavePlayer_4, properties {soundWaveAssetPtr=&/Melee_1HSword_Secondary_Whirlwind_Repeat_01}}
 {id=WavePlayer_5, properties {soundWaveAssetPtr=&/Melee_1HSword_Secondary_Whirlwind_Repeat_02}}
]
properties {
 attenuationSettings=/Game/Sounds/Attenuation/Fort_Weapon_Melee_Attack_Close_Range_Stereo_Att
 maxDistance=10058.75
 soundClassObject=&Fort_WeaponMelee
 volumeMultiplier=0.35
 concurrencySet [/CRD_Melee/Audio/Settings/Melee_Attack_Global_Conc]
}
